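# Niagara idle effect for the SAM Dark one-handed pickaxe that plays a sheen followed by a glint.
# Burst emitter fires once after a short random delay, while Loop emitter repeats at longer random intervals.
# Each uses two spawn groups, timing the glint to the sheen's end, with VisibilityTag set from the group for mesh and sprite renderers.
# System variables share timing, glint color, size and location; curves animate scale, spin, color and material parameters.
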NiagaraSystem /BRCosmetics/Weapons/FORT_Melee/Pickaxe_SAM_Dark_Female_1h/FX/NS_Pickaxe_SAM_Dark_Idle
dependencies {
 usedDataInterfaces [&NiagaraDataInterfaceCurve, &NiagaraDataInterfaceArrayFloat, &NiagaraDataInterfaceColorCurve]
 usedDynamicInputs [
  &Multiply_Vector2
  &RandomRangeVector2D
  &Vector2DFromFloat
 ]
 usedModules [
  &self:SystemSpawnScript.SystemScriptSource.SystemScriptGraph.NiagaraNodeAssignment_1.SetVariables_#1
  &self:SystemSpawnScript.SystemScriptSource.SystemScriptGraph.NiagaraNodeAssignment_0.SetVariables_#2
  &SystemState
 ]
}
emitters {
 Sheen_Ping_Burst {
  emitterUpdateScript [
   &EmitterState(
    Life Cycle Mode=::NewEnumerator1
    Loop Behavior=::NewEnumerator1
    Loop Delay=&RandomRangeFloat(Minimum=0.5, Maximum=1)
    Loop Duration=0.5
    UseLoopDelay=true
   )
   &SpawnBurst_Instantaneous(Use Spawn Probability=false, UseLoopCountLimit=false)
   &SpawnBurst_Instantaneous(Spawn Group=1, Spawn Time=$System.Sheen_Lifetime, Use Spawn Probability=false, UseLoopCountLimit=false) {module=SpawnBurst_Instantaneous001}
  ]
  particleSpawnScript [
   /SetVariables(Particles.VisibilityTag=&SpawnGroup()) {isSetParameters=true, module=SetVariables_#3}
   &InitializeParticle(
    Color=&MaskLinearColorBySpawnGroup(Masked Linear Color=<rgba 1 1 1 1>, Passthrough Linear Color=$System.GlintColor)
    Lifetime=&MaskFloatBySpawnGroup(Masked Float=$System.Sheen_Lifetime, Passthrough Float=&RandomRangeFloat(Minimum=0.3, Maximum=0.7))
    Mesh Scale Mode=::NewEnumerator3
    Position Offset=&MaskVectorBySpawnGroup(Masked Vector=<0 0 1>, Passthrough Vector=$System.GlintLocation)
    Sprite Size=$System.SystemGlintSize
    Sprite Size Mode=::NewEnumerator4
    UsePositionOffset=true
    Write Material Random=true
   )
  ]
  particleUpdateScript [
   &ParticleState()
   &ScaleSpriteSize(Uniform Curve Sprite Scale=DataInterface() {propertyValues {Curve {keys [
        <at 0 0 RCIM_Cubic>
        {arriveTangent=1.6862, interpMode=RCIM_Cubic, leaveTangent=1.6862, time=0.0545, value=0.9401}
        {arriveTangent=-0.9943, interpMode=RCIM_Cubic, leaveTangent=-0.9943, time=0.3437, value=0.5796}
        <at 1 0>
       ]}}})
   &SpriteRotationRate(Rotation Rate=&FloatFromCurve(
     FloatCurve=DataInterface() {propertyValues {Curve {keys [<at 0 1>,<at 1 0>]}}}
     Scale Curve=&Multiply_Float(
      A=140
      B=&SelectFloatFromArray(
       Array Sampling Mode=::NewEnumerator0
       Direct Array Index=?
       Float Selection Array=DataInterface() {propertyValues {FloatData [-1,1]}}
       Random Seed=0
      )
     )
    ))
   &CameraOffset(Camera Offset Amount=&MaskFloatBySpawnGroup(
     Masked Float=1
     Passthrough Float=&FloatFromCurve(FloatCurve=DataInterface() {propertyValues {Curve {keys [<at 0 15>,<at 1 20>]}}}, Scale Curve=2)
    ))
   &ScaleColor(
    Scale Mode=::NewEnumerator1
    Scale RGBA=&MakeVector4_FromLinearColor(LinearColor=&MaskLinearColorBySpawnGroup(
      Masked Linear Color=&ColorFromCurve(ColorCurve=DataInterface() {propertyValues {
         AlphaCurve {keys [<at 0 1>, {arriveTangent=-1, interpMode=RCIM_Cubic, leaveTangent=-1, time=0.5882, value=1}, <at 1 0 RCIM_Cubic>]}
         BlueCurve {keys [<at 0 0.7292 RCIM_Cubic>,<at 0.6045 0.599 RCIM_Cubic>]}
         GreenCurve {keys [<at 0 0.7292 RCIM_Cubic>,<at 0.6045 0.593 RCIM_Cubic>]}
         RedCurve {keys [<at 0 0.7292 RCIM_Cubic>,<at 0.6045 0.4678 RCIM_Cubic>]}
        }})
      Passthrough Linear Color=<rgba 1 1 1 1>
     ))
    ScaleRGBA=true
   )
   &DynamicMaterialParameters(
    Index 0 Param 1=&FloatFromCurve(FloatCurve=DataInterface() {propertyValues {Curve {keys [{arriveTangent=2.513, interpMode=RCIM_Cubic, time=0, value=0.2}, <at 1 1.8>]}}})
    Index 0 Param 2=1
    Module.Index0Param0Write=true
    Module.Index0Param1Write=true
    Module.Index0Param2Write=false
    Module.Index0Param3Write=false
   )
  ]
  rendererClasses [&NiagaraMeshRendererProperties, &NiagaraSpriteRendererProperties]
  renderers [
   {rendererClass=&NiagaraMeshRendererProperties, rendererIndex=0}
   {rendererClass=&NiagaraSpriteRendererProperties, rendererIndex=1}
  ]
 }
 Sheen_Ping_Loop {
  emitterUpdateScript [
   &EmitterState(
    Life Cycle Mode=::NewEnumerator1
    Loop Delay=&RandomRangeFloat(Minimum=2, Maximum=3)
    Loop Duration=&RandomRangeFloat(Minimum=5, Maximum=8)
    UseLoopDelay=true
   )
   &SpawnBurst_Instantaneous(Spawn Time=$System.Sheen_Delay, Use Spawn Probability=false, UseLoopCountLimit=false)
   &SpawnBurst_Instantaneous(
    Spawn Group=1
    Spawn Time=&Add_Float(A=$System.Sheen_Lifetime, B=$System.Sheen_Delay)
    Use Spawn Probability=false
    UseLoopCountLimit=false
   ) {module=SpawnBurst_Instantaneous001}
  ]
  particleSpawnScript [
   /SetVariables(Particles.VisibilityTag=&SpawnGroup()) {isSetParameters=true, module=SetVariables_#3}
   &InitializeParticle(
    Color=&MaskLinearColorBySpawnGroup(Masked Linear Color=<rgba 1 1 1 1>, Passthrough Linear Color=$System.GlintColor)
    Lifetime=&MaskFloatBySpawnGroup(Masked Float=$System.Sheen_Lifetime, Passthrough Float=&RandomRangeFloat(Minimum=0.3, Maximum=0.7))
    Mesh Scale Mode=::NewEnumerator3
    Position Offset=&MaskVectorBySpawnGroup(Masked Vector=<0 0 2>, Passthrough Vector=$System.GlintLocation)
    Sprite Size=$System.SystemGlintSize
    Sprite Size Mode=::NewEnumerator4
    UsePositionOffset=true
    Write Material Random=true
   )
  ]
  particleUpdateScript [
   &ParticleState()
   &ScaleSpriteSize(Uniform Curve Sprite Scale=DataInterface() {propertyValues {Curve {keys [
        <at 0 0 RCIM_Cubic>
        {arriveTangent=1.6862, interpMode=RCIM_Cubic, leaveTangent=1.6862, time=0.0545, value=0.9401}
        {arriveTangent=-0.9943, interpMode=RCIM_Cubic, leaveTangent=-0.9943, time=0.3437, value=0.5796}
        <at 1 0>
       ]}}})
   &SpriteRotationRate(Rotation Rate=&FloatFromCurve(
     FloatCurve=DataInterface() {propertyValues {Curve {keys [<at 0 1>,<at 1 0>]}}}
     Scale Curve=&Multiply_Float(
      A=140
      B=&SelectFloatFromArray(
       Array Sampling Mode=::NewEnumerator0
       Direct Array Index=?
       Float Selection Array=DataInterface() {propertyValues {FloatData [-1,1]}}
       Random Seed=0
      )
     )
    ))
   &CameraOffset(Camera Offset Amount=&MaskFloatBySpawnGroup(
     Masked Float=1
     Passthrough Float=&FloatFromCurve(FloatCurve=DataInterface() {propertyValues {Curve {keys [<at 0 15>,<at 1 20>]}}}, Scale Curve=2)
    ))
   &ScaleColor(
    Scale Mode=::NewEnumerator1
    Scale RGBA=&MakeVector4_FromLinearColor(LinearColor=&MaskLinearColorBySpawnGroup(
      Masked Linear Color=&ColorFromCurve(ColorCurve=DataInterface() {propertyValues {
         AlphaCurve {keys [<at 0 1>, {arriveTangent=-1, interpMode=RCIM_Cubic, leaveTangent=-1, time=0.5882, value=1}, <at 1 0 RCIM_Cubic>]}
         BlueCurve {keys [<at 0 0.7292 RCIM_Cubic>,<at 0.6045 0.599 RCIM_Cubic>]}
         GreenCurve {keys [<at 0 0.7292 RCIM_Cubic>,<at 0.6045 0.593 RCIM_Cubic>]}
         RedCurve {keys [<at 0 0.7292 RCIM_Cubic>,<at 0.6045 0.4678 RCIM_Cubic>]}
        }})
      Passthrough Linear Color=<rgba 1 1 1 1>
     ))
    ScaleRGBA=true
   )
   &DynamicMaterialParameters(
    Index 0 Param 1=&FloatFromCurve(FloatCurve=DataInterface() {propertyValues {Curve {keys [{arriveTangent=2.513, interpMode=RCIM_Cubic, time=0, value=0.2}, <at 1 1.8>]}}})
    Index 0 Param 2=1
    Module.Index0Param0Write=true
    Module.Index0Param1Write=true
    Module.Index0Param2Write=false
    Module.Index0Param3Write=false
   )
  ]
  rendererClasses [&NiagaraMeshRendererProperties, &NiagaraSpriteRendererProperties]
  renderers [
   {rendererClass=&NiagaraMeshRendererProperties, rendererIndex=0}
   {rendererClass=&NiagaraSpriteRendererProperties, rendererIndex=1}
  ]
 }
}
summary {
 systemName=NS_Pickaxe_SAM_Dark_Idle
 emitters [
  {
   emitterName=Sheen_Ping_Loop
   rendererClasses [&NiagaraMeshRendererProperties, &NiagaraSpriteRendererProperties]
  }
  {
   emitterName=Sheen_Ping_Burst
   rendererClasses [&NiagaraMeshRendererProperties, &NiagaraSpriteRendererProperties]
  }
 ]
 userVariables []
}
